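# Sound cue for the wing-flap sound of the DarkVikingFire glider cosmetic in Athena.
# A single wave player feeds the root directly with one flap sample, with no randomization or modulation nodes.
# It uses the shared player glider open/close attenuation and concurrency set, and routes through the glider landing sound class.
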SoundCue /BRCosmetics/Sounds/Gliders/DarkVikingFire/Athena_Glider_DarkVikingFire_Flap_Cue
root=WavePlayer_1
nodes [{id=WavePlayer_1, properties {soundWaveAssetPtr=&/Athena_Glider_DarkVikingFire_Flap_02}}]
properties {
 attenuationSettings=/Game/Sounds/Attenuation/Fort_Player_Glider_OpenClose_Athena_Att
 maxDistance=10000
 soundClassObject=&Fort_Gliders_Land
 volumeMultiplier=1
 concurrencySet [&Athena_GliderOpenClose_Conc]
}
