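# Sound cue for the rattle sound of the DarkVikingFire glider cosmetic.
# A Random node picks one of three equally weighted rattle waves, feeding a Modulator for slight pitch and volume variation.
# It uses glider thrust attenuation and the glider landing sound class.
# A concurrency override caps it at two instances per owner, stopping the oldest, in place of the assigned open/close concurrency set.
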SoundCue /BRCosmetics/Sounds/Gliders/DarkVikingFire/Athena_Glider_DarkVikingFire__Rattle_Cue
root=Modulator_0
nodes [
 {id=Modulator_0, children [Random_1], properties {pitchMax=1, volumeMin=1}}
 {id=Random_1, children [WavePlayer_0,WavePlayer_4,WavePlayer_5], properties {weights [1,1,1]}}
 {id=WavePlayer_0, properties {soundWaveAssetPtr=&/Athena_Glider_DarkVikingFire_Rattle_01}}
 {id=WavePlayer_4, properties {soundWaveAssetPtr=&/Athena_Glider_DarkVikingFire_Rattle_02}}
 {id=WavePlayer_5, properties {soundWaveAssetPtr=&/Athena_Glider_DarkVikingFire_Rattle_03}}
]
properties {
 attenuationSettings=&Fort_Player_Glider_Thrust_Athena_Att
 bOverrideConcurrency=true
 maxDistance=12500
 soundClassObject=&Fort_Gliders_Land
 volumeMultiplier=1
 concurrencyOverrides {bLimitToOwner=true, maxCount=2, resolutionRule=StopOldest}
 concurrencySet [&Athena_GliderOpenClose_Conc]
}
